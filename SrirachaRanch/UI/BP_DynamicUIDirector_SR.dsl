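; Dynamic UI director that shows the streaming radio UI scene while the local player drives a vehicle.
; On BeginPlay it binds to the player controller's enter-as-driver, fully-exit and seat-changed events, and unbinds them on EndPlay.
; The scene is added only if the player is in a zone and the vehicle has a StreamingRadioPlayerComponent.
; Exiting, or switching to a non-driver seat, removes it.
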(event EventBeginPlay
  (|CallParentFunction)
  (bind _asathena_player_controller (Utilities|Casting|CastToAthena_PlayerController (DynamicUIDirector|GetOwningLocalPlayerController)))
  (bind _output_get (Variables|Default|SetAthenaPlayerController _asathena_player_controller))
  (Vehicle|BindEventtoOnEnterVehicleDriver _output_get (EventDispatchers|CreateEvent :!event Custom|Add%20Radio%20Scene))
  (Vehicle|BindEventtoOnFullyExitVehicle (Variables|Default|GetAthenaPlayerController) (EventDispatchers|CreateEvent :!event Custom|Remove%20Radio%20Scene))
  (Vehicle|BindEventtoOnVehicleSeatChanged (Variables|Default|GetAthenaPlayerController) (EventDispatchers|CreateEvent :!event Custom|On%20Vehicle%20Seat%20Changed)))

(event EventEndPlay
  (params (EndPlayReason :type EEndPlayReason%20Enum :default "Destroyed"))
  (|CallParentFunction EndPlayReason)
  (Utilities|IsValid (Variables|Default|GetAthenaPlayerController)
    (:Is%20Valid
      (Vehicle|UnbindEventfromOnEnterVehicleDriver (Variables|Default|GetAthenaPlayerController) (EventDispatchers|CreateEvent :!event Custom|Add%20Radio%20Scene))
      (Vehicle|UnbindEventfromOnFullyExitVehicle (Variables|Default|GetAthenaPlayerController) (EventDispatchers|CreateEvent :!event Custom|Remove%20Radio%20Scene))
      (Vehicle|UnbindEventfromOnVehicleSeatChanged (Variables|Default|GetAthenaPlayerController) (EventDispatchers|CreateEvent :!event Custom|On%20Vehicle%20Seat%20Changed)))))

(event Custom|Add%20Radio%20Scene
  (bind _athenaplayercontroller (Variables|Default|GetAthenaPlayerController))
  (bind _returnvalue (Pawn|GetPlayerPawnorVehicleDriver _athenaplayercontroller))
  (bind _returnvalue_2 (Vehicle|GetVehicleActor _returnvalue :!class /Script/FortniteGame.FortPlayerPawn))
  (Utilities|IsValid _athenaplayercontroller
    (:Is%20Valid
      (Utilities|IsValid _returnvalue
        (:Is%20Valid
          (Utilities|IsValid _returnvalue_2
            (:Is%20Valid
              ;[1 1 1] Conditions for enabling radio scene are:
              ; 1 - Player must be in a zone
              ; 2 - Player's vehicle must have a streaming radio player component
              (if (and (Utilities|IsValid (Actor|GetComponentbyClass _returnvalue_2 "/Script/SrirachaRanch.StreamingRadioPlayerComponent") :!class /Script/Engine.KismetSystemLibrary) (GlobalUIContext|IsinZone (PlayerController|LocalPlayerSubsystems|GetFortGlobalUIContext (Variables|Default|GetAthenaPlayerController))))
                (DynamicUIDirector|AddScene (Variables|Default|GetRadioScene))))))))))

(event Custom|Remove%20Radio%20Scene
  (DynamicUIDirector|RemoveScene (Variables|Default|GetRadioScene)))

(event Custom|On%20Vehicle%20Seat%20Changed
  (params (SeatIndex :type Integer))
  (if (== SeatIndex (Math|Conversions|ToInteger(Byte) (Variables|Default|GetDriverSeat)))
    (CallFunction|AddRadioScene)
    (else
      (CallFunction|RemoveRadioScene))))

(fn ConstructionScript)
