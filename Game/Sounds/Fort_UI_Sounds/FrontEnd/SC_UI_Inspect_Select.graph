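# UI sound cue for an inspect select action in the front end, played in the Fort_UI sound class.
# A mixer layers a main button click with a slightly quieter storage cancel sound.
# Each layer passes through its own modulator, and the storage cancel layer is pitched slightly up.
# A concurrency override allows one instance at a time, stopping the oldest when retriggered.
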SoundCue /Game/Sounds/Fort_UI_Sounds/FrontEnd/SC_UI_Inspect_Select
root=Mixer_0
nodes [
 {id=Mixer_0, children [Modulator_0,Modulator_1], properties {inputVolume [1,0.8]}}
 {id=Modulator_0, children [WavePlayer_7]}
 {id=Modulator_1, children [WavePlayer_0], properties {pitchMax=1.1, pitchMin=1.08}}
 {id=WavePlayer_0, properties {soundWaveAssetPtr=&/Fort_UI_Storage_Cancel}}
 {id=WavePlayer_7, properties {soundWaveAssetPtr=/Game/Sounds/Fort_UI_Sounds/FrontEnd/New_FrontEnd/UI_FN_Button_Click_Main_01}}
]
properties {
 bOverrideConcurrency=true
 maxDistance=2097152
 soundClassObject=&Fort_UI
 volumeMultiplier=1
 concurrencyOverrides {maxCount=1, resolutionRule=StopOldest}
}
