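# Radio menu radial picker entry widget, derived from FortRadialPickerEntry, that adds visual and controller feedback on selection.
# Selecting plays a short icon scale-up animation and a force feedback effect; deselecting plays the animation in reverse.
# When bUpdateEntryColorOnSelection is true, BP_UpdateSelectionColor tints the icon with the selected color only if selected and enabled.
# Availability changes also call that function, passing availability as the selected state.
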WidgetBlueprint /SrirachaRanch/UI/RadioMenuPickerEntry
animations [Selected {
  binding_count=1
  display_label=""
  length_seconds=0.1
  object=&self:Selected
  sequence {bindings [Image_Icon {
     id=#1
     possessed_class=CommonLazyImage
     tracks [{
       type=MovieScene2DTransformTrack
       sections [{
         type=MovieScene2DTransformSection
         channels [
          Translation.X {type=MovieSceneScriptingFloatChannel}
          Translation.Y {type=MovieSceneScriptingFloatChannel}
          Angle {type=MovieSceneScriptingFloatChannel}
          Scale.X {type=MovieSceneScriptingFloatChannel, keys [[0,0.9],[2,1.2]]}
          Scale.Y {type=MovieSceneScriptingFloatChannel, keys [[0,0.9],[2,1.2]]}
          Shear.X {type=MovieSceneScriptingFloatChannel}
          Shear.Y {type=MovieSceneScriptingFloatChannel}
         ]
        }]
      }]
    }]}
 }]
graphs [
 BP_UpdateSelectionColor {nodes [
   n1 = BP_UpdateSelectionColor!() -> (then, Selected:bool=~) {is=title K2Node_FunctionEntry}
   > Branch(Condition=Get bUpdateEntryColorOnSelection() -> (bUpdateEntryColorOnSelection:bool))
   n2 = AND Boolean(A=@n1, B=GetIsEnabled!(self:Widget Object Reference) -> (ReturnValue:bool))
   n3 = Get DefaultUnselectedColor() -> (DefaultUnselectedColor:Linear Color Structure)
   n4 = Get DefaultSelectedColor() -> (DefaultSelectedColor:Linear Color Structure)
   > SetBrushTintColor!(
    execute
    self:Image Object Reference=Get Image_Icon() -> (Image_Icon:Common Lazy Image Object Reference)
    TintColor:Slate Color Structure=Make Slate Color(SpecifiedColor=Select(Option 0:Linear Color Structure=@n3, Option 1:Linear Color Structure=@n4, Index=@n2) -> (ReturnValue:Linear Color Structure))
   ) -> (then)
  ]}
 EventGraph {links [], nodes []}
 Selection {nodes [
   Event On Selected()
   > PlayAnimation(InAnimation=Get Selected() -> (Selected:Widget Animation Object Reference))
   > BP_UpdateSelectionColor!(execute, self, Selected:bool=true) -> (then) {is=title K2Node_CallFunction}
   > Client Play Force Feedback!(
    execute
    self:Player Controller Object Reference=GetOwningPlayer()
    ForceFeedbackEffect:Force Feedback Effect Object Reference=/Game/UI/InGame/HUD/AthenaRadial/FF_RadialItemSelection
    Tag:Name=None
    bLooping:bool=false
    bIgnoreTimeDilation:bool=false
    bPlayWhilePaused:bool=false
   ) -> (then)
   Event On Deselected()
   > PlayAnimation(InAnimation=Get Selected(), PlayMode=Reverse)
   > BP_UpdateSelectionColor(Selected=false)
   n1 = Event OnAvailabilityChanged!() -> (OutputDelegate, then, bIsAvailable:bool) {is=title K2Node_Event}
   > BP_UpdateSelectionColor(Selected=@n1.bIsAvailable)
  ]}
]
variables [
 bUpdateEntryColorOnSelection {type=boolean}
 DefaultSelectedColor {type=LinearColor, default=<rgba 0 0 0 1>}
 DefaultUnselectedColor {type=LinearColor, default=<rgba 1 1 1 1>}
]
widgets {
 info {inheritedWidgetCount=8, parentClass=/Script/FortniteUI.FortRadialPickerEntry, rootWidgetClass=&Overlay, widgetCount=8}
 widgets [
  Overlay_EquippedItem {widgetClassPath=&Overlay, bInherited=true, widget=Overlay_EquippedItem}
  Overlay_SizeBox {widgetClassPath=&SizeBox, bInherited=true, parent=Overlay_EquippedItem, widget=Overlay_SizeBox, slot {class=OverlaySlot}}
  Image_Icon {
   widgetClassPath=/Script/CommonUI.CommonLazyImage
   bInherited=true
   bIsVariable=true
   parent=Overlay_SizeBox
   widget=Image_Icon
   properties {brush {
     bIsDynamicallyLoaded=false
     drawAs=Image
     imageType=NoImage
     mirroring=NoMirror
     resourceName=None
     resourceObject=/Game/UI/Foundation/Textures/Icons/Weapons/Items/T-Icon-Weapons-TacticalShotgunMark2Base
     tiling=NoTile
     imageSize=<512 512>
     margin=<ltrb 0 0 0 0>
     outlineSettings {
      bUseBrushTransparency=false
      roundingType=HalfHeightRadius
      width=0
      color {colorUseRule=UseColor_Specified, specifiedColor=<rgba 0 0 0 0>}
      cornerRadii=<0 0 0 0>
     }
     tintColor {colorUseRule=UseColor_Specified, specifiedColor=<rgba 0.3864 0.3864 0.3864 1>}
     uVRegion {bIsValid=false, max=<0 0>, min=<0 0>}
    }}
   slot {class=SizeBoxSlot}
  }
  HorizontalBox_EntryName {widgetClassPath=&HorizontalBox, bInherited=true, widget=null}
  Text_Center {widgetClassPath=&CommonTextBlock, bInherited=true, widget=null}
  HorizontalBox_CenterText {widgetClassPath=&HorizontalBox, bInherited=true, widget=null}
  Text_Main {widgetClassPath=&CommonTextBlock, bInherited=true, widget=null}
  InputActionWidget {widgetClassPath=&CommonActionWidget, bInherited=true, widget=null}
 ]
}
