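# Radio menu picker overlay, a FortPickerOverlay widget that shows a centred radial wheel of RadioMenuPickerEntry items.
# A FortSlottedRadialMenu, steered by left stick or mouse with dead zones, sits over pie-material background, shadow and highlight images.
# Activation plays a scale-and-fade Intro with a sound, deactivation reverses it faster, and choosing an item plays a sound.
# PlaySelectionAnimation runs SelectionSet, which fades in the highlight and animates its material's OuterRadius parameter.
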WidgetBlueprint /SrirachaRanch/UI/RadioMenuPickerOverlay
animations [
 SelectionSet {
  binding_count=1
  length_seconds=0.2
  object=&self:SelectionSet
  sequence {bindings [Image_RadialHighlight {
     id=#1
     possessed_class=Image
     tracks [
      {
       type=MovieSceneWidgetMaterialTrack
       sections [{
         type=MovieSceneParameterSection
         channels [
          OuterRadius_0-1 {type=MovieSceneScriptingFloatChannel, keys [[0,0.9],[4,1]]}
          InnerRadius_0-1 {type=MovieSceneScriptingFloatChannel}
         ]
        }]
      }
      {
       type=MovieSceneFloatTrack
       display_name=Render Opacity
       sections [{type=MovieSceneFloatSection, channels [None {type=MovieSceneScriptingFloatChannel, keys [[0,0],[1.557,1]]}]}]
      }
     ]
    }]}
 }
 Intro {
  binding_count=2
  length_seconds=0.2
  object=&self:Intro
  sequence {
   bindings [
    Overlay_1 {
     id=#2
     possessed_class=Overlay
     tracks [
      {
       type=MovieScene2DTransformTrack
       sections [{
         type=MovieScene2DTransformSection
         channels [
          Translation.X {type=MovieSceneScriptingFloatChannel}
          Translation.Y {type=MovieSceneScriptingFloatChannel}
          Angle {type=MovieSceneScriptingFloatChannel}
          Scale.X {type=MovieSceneScriptingFloatChannel, keys [[0,0],[4,1]]}
          Scale.Y {type=MovieSceneScriptingFloatChannel, keys [[0,0],[4,1]]}
          Shear.X {type=MovieSceneScriptingFloatChannel}
          Shear.Y {type=MovieSceneScriptingFloatChannel}
         ]
        }]
      }
      {
       type=MovieSceneFloatTrack
       display_name=Render Opacity
       sections [{type=MovieSceneFloatSection, channels [None {type=MovieSceneScriptingFloatChannel, keys [[0,0],[2,1]]}]}]
      }
     ]
    }
    RadialMenu_Picker {
     id=#3
     possessed_class=FortSlottedRadialMenu_C
     tracks [
      {
       type=MovieSceneFloatTrack
       display_name=Render Opacity
       sections [{type=MovieSceneFloatSection, channels [None {type=MovieSceneScriptingFloatChannel, keys [[0,0],[1,0],[4.0003,1]]}]}]
      }
      {
       type=MovieScene2DTransformTrack
       sections [{
         type=MovieScene2DTransformSection
         end_frame=4
         start_frame=0
         channels [
          Scale.X {type=MovieSceneScriptingFloatChannel, keys [[1,0.75],[4,1]]}
          Scale.Y {type=MovieSceneScriptingFloatChannel, keys [[1,0.75],[4,1]]}
         ]
        }]
      }
     ]
    }
   ]
   tracks [{
     type=MovieSceneEventTrack
     display_name=IntroFinished
     sections [{
       type=MovieSceneEventTriggerSection
       channels [None {type=MovieSceneScriptingEventChannel, const ["<Struct 'MovieSceneEvent' (0x46944636e00) {}>", 1]}]
      }]
    }]
  }
 }
]
graphs [
 Cosmetic {nodes [
   Event On Activated()
   > PlayAnimationForward(InAnimation=Get Intro(), PlaybackSpeed=1.5)
   > PlaySound2D(Sound=/Game/Sounds/Fort_UI_Sounds/FrontEnd/SC_UI_Inspect_Select)
   Event On Item Chosen() {is=title K2Node_Event}
   > PlaySound2D(Sound=/Game/Sounds/Fort_UI_Sounds/FrontEnd/New_FrontEnd/Fortnite_UI_MainMenu_Generic_03_Cue)
   Event On Deactivated()
   > PlayAnimationReverse(InAnimation=Get Intro(), PlaybackSpeed=3)
   Event PlaySelectionAnimation() {is=title K2Node_Event}
   > PlayAnimation(InAnimation=Get SelectionSet() -> (SelectionSet:Widget Animation Object Reference))
  ]}
 EventGraph {links [], nodes []}
]
variables []
widgets {
 info {inheritedWidgetCount=3, parentClass=/Script/FortniteUI.FortPickerOverlay, rootWidgetClass=&Overlay, widgetCount=7}
 widgets [
  Overlay_1 {widgetClassPath=&Overlay, widget=Overlay_1}
  SizeBox_7 {
   widgetClassPath=&SizeBox
   parent=Overlay_1
   widget=SizeBox_7
   properties {bOverride_HeightOverride=true, bOverride_WidthOverride=true, heightOverride=700, widthOverride=700}
   slot {class=OverlaySlot, properties {horizontalAlignment=HAlign_Center, verticalAlignment=VAlign_Center}}
  }
  Overlay_2 {widgetClassPath=&Overlay, parent=SizeBox_7, widget=Overlay_2, slot {class=SizeBoxSlot}}
  Image_Radial_BG {
   widgetClassPath=&Image
   parent=Overlay_2
   widget=Image_Radial_BG
   properties {
    visibility=HitTestInvisible
    brush {
     bIsDynamicallyLoaded=false
     drawAs=Image
     imageType=NoImage
     mirroring=NoMirror
     resourceName=None
     resourceObject=/Game/UI/Items/Materials/MI-AthenaPieShadow
     tiling=NoTile
     imageSize=<128 128>
     margin=<ltrb 0 0 0 0>
     outlineSettings {
      bUseBrushTransparency=false
      roundingType=HalfHeightRadius
      width=0
      color {colorUseRule=UseColor_Specified, specifiedColor=<rgba 0 0 0 0>}
      cornerRadii=<0 0 0 0>
     }
     tintColor {colorUseRule=UseColor_Specified, specifiedColor=<rgba 1 1 1 1>}
     uVRegion {bIsValid=false, max=<0 0>, min=<0 0>}
    }
    colorAndOpacity=<rgba 0.0052 0.0052 0.0056 0.4>
    renderTransform {angle=0, scale=<1 1>, shear=<0 0>, translation=<20 20>}
   }
   slot {class=OverlaySlot, properties {horizontalAlignment=HAlign_Fill, verticalAlignment=VAlign_Fill}}
  }
  Image_Radial {
   widgetClassPath=&Image
   bInherited=true
   parent=Overlay_2
   widget=Image_Radial
   properties {
    visibility=HitTestInvisible
    brush {
     bIsDynamicallyLoaded=false
     drawAs=Image
     imageType=NoImage
     mirroring=NoMirror
     resourceName=None
     resourceObject=/Game/UI/Items/Materials/MI-AthenaPieMenu
     tiling=NoTile
     imageSize=<128 128>
     margin=<ltrb 0 0 0 0>
     outlineSettings {
      bUseBrushTransparency=false
      roundingType=HalfHeightRadius
      width=0
      color {colorUseRule=UseColor_Specified, specifiedColor=<rgba 0 0 0 0>}
      cornerRadii=<0 0 0 0>
     }
     tintColor {colorUseRule=UseColor_Specified, specifiedColor=<rgba 1 1 1 1>}
     uVRegion {bIsValid=false, max=<0 0>, min=<0 0>}
    }
    colorAndOpacity=<rgba 0.0273 0.0273 0.0273 1>
   }
   slot {class=OverlaySlot, properties {horizontalAlignment=HAlign_Fill, verticalAlignment=VAlign_Fill}}
  }
  Image_RadialHighlight {
   widgetClassPath=&Image
   bInherited=true
   parent=Overlay_2
   widget=Image_RadialHighlight
   properties {
    visibility=HitTestInvisible
    brush {
     bIsDynamicallyLoaded=false
     drawAs=Image
     imageType=FullColor
     mirroring=NoMirror
     resourceName=None
     resourceObject=/Game/UI/Items/Materials/MI-AthenaPieMenu_Selection
     tiling=NoTile
     imageSize=<8589936640 0>
     margin=<ltrb 0 0 0 0>
     outlineSettings {
      bUseBrushTransparency=false
      roundingType=HalfHeightRadius
      width=0
      color {colorUseRule=UseColor_Specified, specifiedColor=<rgba 0 0 0 0>}
      cornerRadii=<0 0 0 0>
     }
     tintColor {colorUseRule=UseColor_Specified, specifiedColor=<rgba 1 1 1 1>}
     uVRegion {bIsValid=false, max=<0 0>, min=<0 0>}
    }
    renderTransform {angle=0, scale=<1.0325 1.0325>, shear=<0 0>, translation=<0 0>}
   }
   slot {class=OverlaySlot, properties {horizontalAlignment=HAlign_Fill, verticalAlignment=VAlign_Fill, padding=<ltrb 2 2 2 2>}}
  }
  RadialMenu_Picker {
   widgetClassPath=/Game/Athena/HUD/ContextMenus/FortSlottedRadialMenu
   bInherited=true
   parent=Overlay_2
   widget=RadialMenu_Picker
   properties {
    controllingStick=Left
    entryWidgetClass=&/RadioMenuPickerEntry.RadioMenuPickerEntry_C
    joystickDeadZone=0.4
    mouseDeadZoneRadiusPercentage=0.3
    navigation=&self:WidgetTree.RadialMenu_Picker.WidgetNavigation_0
    orderingMode=Clockwise
    visibility=Visible
    radialBoxSettings {angleBetweenItems=0, bClockwiseOrder=false, bDistributeItemsEvenly=true, marginSize=0, sectorCentralAngle=360, startingAngle=90}
   }
   slot {class=OverlaySlot, properties {horizontalAlignment=HAlign_Fill, verticalAlignment=VAlign_Fill}}
  }
 ]
}
