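# Widget Blueprint for the radio station picker wheel, built on the native RadioMenuPicker class.
# It lays out a picker overlay with dynamically sized slots, station name, artist and song text, a glow, and a bottom bar of input hints.
# PreConstruct toggles the details panel, hints and glow from bool flags.
# Activation resets root opacity and translation and plays an Intro fade; deactivation stops all animations.
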WidgetBlueprint /SrirachaRanch/UI/RadioMenuPicker
animations [Intro {
  binding_count=2
  length_seconds=2.9885
  object=&self:Intro
  sequence {bindings [
    Overlay_Main {
     id=#1
     possessed_class=Overlay
     tracks [{
       type=MovieSceneFloatTrack
       display_name=Render Opacity
       sections [{type=MovieSceneFloatSection, channels [None {type=MovieSceneScriptingFloatChannel, keys [[0.0077,0],[1.0773,0],[3.246,1]]}]}]
      }]
    }
    Image_BGGlow {
     id=#2
     possessed_class=Image
     tracks [{
       type=MovieSceneFloatTrack
       display_name=Render Opacity
       sections [{type=MovieSceneFloatSection, channels [None {type=MovieSceneScriptingFloatChannel, keys [[1.0773,0],[29.9923,0.9604],[59.7773,1]]}]}]
      }]
    }
   ]}
 }]
graphs [
 EventGraph {nodes [
   Event On Deactivated()
   > StopAllAnimations(self:User Widget Object Reference)
   Event On Activated()
   > ResetAnimationState()
   > PlayIntroAnimation()
   Event PreConstruct()
   > SetVisibility(
    self=Get Panel_RadioStationDetails() -> (Panel_RadioStationDetails:Vertical Box Object Reference)
    InVisibility="To Visibility (Boolean)"!(
     bIsVisible:bool=Get bShowRadioStationDetailsPanel() -> (bShowRadioStationDetailsPanel:bool)
     TrueVisibility:ESlateVisibility Enum=HitTestInvisible
     FalseVisibility:ESlateVisibility Enum=Collapsed
    ) -> (ReturnValue:ESlateVisibility Enum=Visible)
   )
   > SetVisibility(
    self=Get CommonVisibilityWidget_InputHints() -> (CommonVisibilityWidget_InputHints:Common Visibility Widget Legacy Object Reference)
    InVisibility="To Visibility (Boolean)"(bIsVisible=Get bShowInputHintsBottomBar() -> (bShowInputHintsBottomBar:bool), TrueVisibility=HitTestInvisible, FalseVisibility=Collapsed) -> (ReturnValue=Visible)
   )
   > SetVisibility(
    self=Get Image_BGGlow() -> (Image_BGGlow:Image Object Reference)
    InVisibility="To Visibility (Boolean)"(bIsVisible=Get bShowBackgroundGlow() -> (bShowBackgroundGlow:bool), TrueVisibility=HitTestInvisible, FalseVisibility=Collapsed) -> (ReturnValue=Visible)
   )
  ]}
 PlayIntroAnimation {nodes [
   PlayIntroAnimation() {is=title K2Node_FunctionEntry}
   > PlayAnimation(InAnimation=Get Intro())
  ]}
 ResetAnimationState {nodes [
   ResetAnimationState() {is=title K2Node_FunctionEntry}
   > SetRenderOpacity(InOpacity=1)
   > SetRenderTranslation()
  ]}
]
variables []
widgets {
 info {inheritedWidgetCount=9, parentClass=/Script/SrirachaRanchUI.RadioMenuPicker, rootWidgetClass=&Overlay, widgetCount=34}
 widgets [
  Overlay_Main {widgetClassPath=&Overlay, widget=Overlay_Main}
  ClickCatcher {
   widgetClassPath=&Spacer
   parent=Overlay_Main
   widget=ClickCatcher
   properties {visibility=Visible}
   slot {class=OverlaySlot, properties {horizontalAlignment=HAlign_Fill, verticalAlignment=VAlign_Fill}}
  }
  Image_BGGlow {
   widgetClassPath=&Image
   bIsVariable=true
   parent=Overlay_Main
   widget=Image_BGGlow
   properties {
    visibility=SelfHitTestInvisible
    brush {
     bIsDynamicallyLoaded=false
     drawAs=Image
     imageType=NoImage
     mirroring=NoMirror
     resourceName=None
     resourceObject=/Game/Athena/HUD/Materials/UI_RebootAlert_Glow_Inst
     tiling=NoTile
     imageSize=<2000 2000>
     margin=<ltrb 0 0 0 0>
     outlineSettings {
      bUseBrushTransparency=false
      roundingType=HalfHeightRadius
      width=0
      color {colorUseRule=UseColor_Specified, specifiedColor=<rgba 0 0 0 0>}
      cornerRadii=<0 0 0 0>
     }
     tintColor {colorUseRule=UseColor_Specified, specifiedColor=<rgba 0.0666 0.5647 0.9647 1>}
     uVRegion {bIsValid=false, max=<0 0>, min=<0 0>}
    }
    colorAndOpacity=<rgba 1 1 1 0.25>
   }
   slot {class=OverlaySlot, properties {horizontalAlignment=HAlign_Center, verticalAlignment=VAlign_Center}}
  }
  VerticalBox_0 {
   widgetClassPath=&VerticalBox
   parent=Overlay_Main
   widget=VerticalBox_0
   slot {class=OverlaySlot, properties {horizontalAlignment=HAlign_Center, verticalAlignment=VAlign_Center}}
  }
  Panel_RadioStationDetails {
   widgetClassPath=&VerticalBox
   bIsVariable=true
   parent=VerticalBox_0
   widget=Panel_RadioStationDetails
   slot {class=VerticalBoxSlot, properties {horizontalAlignment=HAlign_Center, verticalAlignment=VAlign_Bottom}}
  }
  Text_RadioStationName {
   widgetClassPath=&CommonTextBlock
   bInherited=true
   parent=Panel_RadioStationDetails
   widget=Text_RadioStationName
   properties {
    justification=Center
    lineHeightPercentage=0.8
    text=Party Royale
    textTransformPolicy=ToUpper
    font {
     bForceMonospaced=false
     bMaterialIsStencil=false
     fontMaterial=null
     fontObject=&Burbank
     letterSpacing=0
     monospacedWidth=1
     size=43
     skewAmount=0
     typefaceFontName=Burbank-PlayerPower
     outlineSettings {
      bApplyOutlineToDropShadows=false
      bMiteredCorners=false
      bSeparateFillAlpha=false
      outlineMaterial=&M-FontShadow
      outlineSize=3
      outlineColor=<rgba 0 0 0 0.5>
     }
    }
    shadowColorAndOpacity=<rgba 0 0 0 0.5>
    shadowOffset=<5 5>
   }
   slot {class=VerticalBoxSlot}
  }
  Text_RadioStationCurrentArtist {
   widgetClassPath=&CommonTextBlock
   bInherited=true
   parent=Panel_RadioStationDetails
   widget=Text_RadioStationCurrentArtist
   properties {
    justification=Center
    lineHeightPercentage=0.8
    text=Epic Games
    textTransformPolicy=ToUpper
    colorAndOpacity {colorUseRule=UseColor_Specified, specifiedColor=<rgba 1 0.8725 0.3209 1>}
    font {
     bForceMonospaced=false
     bMaterialIsStencil=false
     fontMaterial=null
     fontObject=&BurbankSmall
     letterSpacing=0
     monospacedWidth=1
     size=28
     skewAmount=0
     typefaceFontName=Bold
     outlineSettings {
      bApplyOutlineToDropShadows=false
      bMiteredCorners=false
      bSeparateFillAlpha=false
      outlineMaterial=&M-FontShadow
      outlineSize=3
      outlineColor=<rgba 0 0 0 0.5>
     }
    }
    shadowColorAndOpacity=<rgba 0 0 0 0.5>
    shadowOffset=<5 5>
    strikeBrush {
     bIsDynamicallyLoaded=false
     drawAs=Image
     imageType=NoImage
     mirroring=NoMirror
     resourceName=None
     resourceObject=null
     tiling=NoTile
     imageSize=<32 32>
     margin=<ltrb 0 0 0 0>
     outlineSettings {
      bUseBrushTransparency=false
      roundingType=HalfHeightRadius
      width=0
      color {colorUseRule=UseColor_Specified, specifiedColor=<rgba 0 0 0 0>}
      cornerRadii=<0 0 0 0>
     }
     tintColor {colorUseRule=UseColor_Specified, specifiedColor=<rgba 0 0 0 1>}
     uVRegion {bIsValid=false, max=<0 0>, min=<0 0>}
    }
   }
   slot {class=VerticalBoxSlot, properties {horizontalAlignment=HAlign_Center, verticalAlignment=VAlign_Top}}
  }
  Text_RadioStationCurrentSong {
   widgetClassPath=&CommonTextBlock
   bInherited=true
   parent=Panel_RadioStationDetails
   widget=Text_RadioStationCurrentSong
   properties {
    justification=Center
    lineHeightPercentage=0.85
    text=When the wind blows
    textTransformPolicy=ToUpper
    colorAndOpacity {colorUseRule=UseColor_Specified, specifiedColor=<rgba 1 1 1 0.8>}
    font {
     bForceMonospaced=false
     bMaterialIsStencil=false
     fontMaterial=null
     fontObject=&BurbankSmall
     letterSpacing=0
     monospacedWidth=1
     size=28
     skewAmount=0
     typefaceFontName=Bold
     outlineSettings {
      bApplyOutlineToDropShadows=false
      bMiteredCorners=false
      bSeparateFillAlpha=false
      outlineMaterial=&M-FontShadow
      outlineSize=3
      outlineColor=<rgba 0 0 0 0.5>
     }
    }
    shadowColorAndOpacity=<rgba 0 0 0 0.5>
    shadowOffset=<5 5>
    strikeBrush {
     bIsDynamicallyLoaded=false
     drawAs=Image
     imageType=NoImage
     mirroring=NoMirror
     resourceName=None
     resourceObject=null
     tiling=NoTile
     imageSize=<32 32>
     margin=<ltrb 0 0 0 0>
     outlineSettings {
      bUseBrushTransparency=false
      roundingType=HalfHeightRadius
      width=0
      color {colorUseRule=UseColor_Specified, specifiedColor=<rgba 0 0 0 0>}
      cornerRadii=<0 0 0 0>
     }
     tintColor {colorUseRule=UseColor_Specified, specifiedColor=<rgba 0 0 0 1>}
     uVRegion {bIsValid=false, max=<0 0>, min=<0 0>}
    }
   }
   slot {class=VerticalBoxSlot, properties {horizontalAlignment=HAlign_Center, verticalAlignment=VAlign_Top}}
  }
  Overlay_229 {widgetClassPath=&Overlay, parent=VerticalBox_0, widget=Overlay_229, slot {class=VerticalBoxSlot}}
  PickerOverlay_RadioMenu {
   widgetClassPath=&/RadioMenuPickerOverlay
   bInherited=true
   bIsVariable=true
   parent=Overlay_229
   widget=PickerOverlay_RadioMenu
   properties {
    bAllowDynamicNumberOfSlots=true
    colorAndOpacity=<rgba 1 1 1 1>
    slotPropertiesForDynamicNumberOfSlots [
     {imageSize=<160 160>, padding=<ltrb 0 0 0 0>}
     {imageSize=<160 160>, padding=<ltrb 0 0 0 0>}
     {imageSize=<160 160>, padding=<ltrb 0 0 0 0>}
     {imageSize=<160 160>, padding=<ltrb 0 0 0 0>}
     {imageSize=<160 160>, padding=<ltrb 0 0 0 0>}
     {imageSize=<160 160>, padding=<ltrb 0 0 0 0>}
     {imageSize=<160 160>, padding=<ltrb 0 0 0 0>}
     {imageSize=<160 160>, padding=<ltrb 0 0 0 0>}
     {imageSize=<150 150>, padding=<ltrb 5 5 5 5>}
     {imageSize=<150 150>, padding=<ltrb 5 5 5 5>}
     {imageSize=<140 140>, padding=<ltrb 10 10 10 10>}
     {imageSize=<140 140>, padding=<ltrb 10 10 10 10>}
     {imageSize=<130 130>, padding=<ltrb 15 15 15 15>}
     {imageSize=<120 120>, padding=<ltrb 20 20 20 20>}
     {imageSize=<110 110>, padding=<ltrb 25 25 25 25>}
     {imageSize=<100 100>, padding=<ltrb 30 30 30 30>}
    ]
   }
   slot {class=OverlaySlot, properties {horizontalAlignment=HAlign_Center, verticalAlignment=VAlign_Center}}
  }
  SizeBox {
   widgetClassPath=&SizeBox
   parent=Overlay_229
   widget=SizeBox
   properties {bIsEnabled=false, bOverride_HeightOverride=true, bOverride_WidthOverride=true, heightOverride=50, renderOpacity=0, widthOverride=50}
   slot {class=OverlaySlot, properties {horizontalAlignment=HAlign_Center, verticalAlignment=VAlign_Center}}
  }
  ToggleActionWidget {
   widgetClassPath=&CommonActionWidget
   parent=SizeBox
   widget=ToggleActionWidget
   properties {
    bIsEnabled=false
    progressMaterialParam=Percentage
    inputActions [{dataTable=&/RadioInputActionTable, rowName=ReleasePicker}]
    progressMaterialBrush {
     bIsDynamicallyLoaded=false
     drawAs=Image
     imageType=NoImage
     mirroring=NoMirror
     resourceName=None
     resourceObject=/Game/Athena/HUD/M_HealProgress
     tiling=NoTile
     imageSize=<80 80>
     margin=<ltrb 0 0 0 0>
     outlineSettings {
      bUseBrushTransparency=false
      roundingType=HalfHeightRadius
      width=0
      color {colorUseRule=UseColor_Specified, specifiedColor=<rgba 0 0 0 0>}
      cornerRadii=<0 0 0 0>
     }
     tintColor {colorUseRule=UseColor_Specified, specifiedColor=<rgba 1 1 1 1>}
     uVRegion {bIsValid=false, max=<0 0>, min=<0 0>}
    }
    renderTransformPivot=<1 1>
   }
   slot {class=SizeBoxSlot, properties {horizontalAlignment=HAlign_Center, verticalAlignment=VAlign_Center}}
  }
  CommonVisibilityWidget_InputHints {
   widgetClassPath=&CommonVisibilityWidgetLegacy
   bIsVariable=true
   parent=Overlay_Main
   widget=CommonVisibilityWidget_InputHints
   properties {bShowForTouch=false, padding=<ltrb 0 0 0 0>}
   slot {class=OverlaySlot, properties {horizontalAlignment=HAlign_Fill, verticalAlignment=VAlign_Bottom}}
  }
  BottomBarOverlay {
   widgetClassPath=&Overlay
   parent=CommonVisibilityWidget_InputHints
   widget=BottomBarOverlay
   slot {class=BorderSlot, properties {padding=<ltrb 0 0 0 0>}}
  }
  Overlay {
   widgetClassPath=&Overlay
   parent=BottomBarOverlay
   widget=Overlay
   slot {class=OverlaySlot, properties {horizontalAlignment=HAlign_Fill, verticalAlignment=VAlign_Fill}}
  }
  Image_bottombar {
   widgetClassPath=&Image
   parent=Overlay
   widget=Image_bottombar
   properties {
    brush {
     bIsDynamicallyLoaded=false
     drawAs=Image
     imageType=NoImage
     mirroring=NoMirror
     resourceName=None
     resourceObject=null
     tiling=NoTile
     imageSize=<32 45>
     margin=<ltrb 0 0 0 0>
     outlineSettings {
      bUseBrushTransparency=false
      roundingType=HalfHeightRadius
      width=0
      color {colorUseRule=UseColor_Specified, specifiedColor=<rgba 0 0 0 0>}
      cornerRadii=<0 0 0 0>
     }
     tintColor {colorUseRule=UseColor_Specified, specifiedColor=<rgba 1 1 1 1>}
     uVRegion {bIsValid=false, max=<0 0>, min=<0 0>}
    }
    colorAndOpacity=<rgba 0.006 0.011 0.0242 0.8>
   }
   slot {class=OverlaySlot, properties {horizontalAlignment=HAlign_Fill, verticalAlignment=VAlign_Fill}}
  }
  VerticalBox {
   widgetClassPath=&VerticalBox
   parent=Overlay
   widget=VerticalBox
   slot {class=OverlaySlot, properties {horizontalAlignment=HAlign_Fill, verticalAlignment=VAlign_Fill}}
  }
  SafeZone {
   widgetClassPath=&SafeZone
   parent=VerticalBox
   widget=SafeZone
   properties {padBottom=false, padLeft=false, padRight=false}
   slot {class=VerticalBoxSlot}
  }
  Spacer {widgetClassPath=&Spacer, parent=VerticalBox, widget=Spacer, properties {size=<1 45>}, slot {class=VerticalBoxSlot}}
  SizeBox_11 {
   widgetClassPath=&SizeBox
   parent=BottomBarOverlay
   widget=SizeBox_11
   properties {bOverride_WidthOverride=true, widthOverride=900}
   slot {class=OverlaySlot, properties {horizontalAlignment=HAlign_Center, verticalAlignment=VAlign_Bottom}}
  }
  HorizontalBox_4 {widgetClassPath=&HorizontalBox, parent=SizeBox_11, widget=HorizontalBox_4, slot {class=SizeBoxSlot}}
  Overlay_InGameEmoteBinding {
   widgetClassPath=&Overlay
   parent=HorizontalBox_4
   widget=Overlay_InGameEmoteBinding
   slot {class=HorizontalBoxSlot, properties {horizontalAlignment=HAlign_Center, verticalAlignment=VAlign_Center, size {sizeRule=Fill, value=1}}}
  }
  EmoteHotkey {
   widgetClassPath=&HorizontalBox
   parent=Overlay_InGameEmoteBinding
   widget=EmoteHotkey
   slot {class=OverlaySlot, properties {horizontalAlignment=HAlign_Center}}
  }
  SizeBox_1 {
   widgetClassPath=&SizeBox
   parent=EmoteHotkey
   widget=SizeBox_1
   properties {bOverride_HeightOverride=true, heightOverride=40, widthOverride=30}
   slot {class=HorizontalBoxSlot}
  }
  Keybind_Confirm {
   widgetClassPath=/FortUILibrary/Input/KeybindWidget
   bInherited=true
   parent=SizeBox_1
   widget=Keybind_Confirm
   properties {forcedHoldKeybindStatus=NeverShowHold}
   slot {class=SizeBoxSlot, properties {padding=<ltrb 0 0 10 0>}}
  }
  CommonTextBlock {
   widgetClassPath=&CommonTextBlock
   parent=EmoteHotkey
   widget=CommonTextBlock
   properties {
    justification=Center
    text=Select
    font {
     bForceMonospaced=false
     bMaterialIsStencil=false
     fontMaterial=null
     fontObject=&BurbankSmall
     letterSpacing=0
     monospacedWidth=1
     size=18
     skewAmount=0
     typefaceFontName=Bold
     outlineSettings {
      bApplyOutlineToDropShadows=false
      bMiteredCorners=false
      bSeparateFillAlpha=false
      outlineMaterial=null
      outlineSize=0
      outlineColor=<rgba 0 0 0 1>
     }
    }
    shadowOffset=<0 0>
   }
   slot {class=HorizontalBoxSlot, properties {horizontalAlignment=HAlign_Center, verticalAlignment=VAlign_Center}}
  }
  Button_Cancel {
   widgetClassPath=/FortUILibrary/Input/InputReflectorButton
   bInherited=true
   bIsVariable=true
   parent=HorizontalBox_4
   widget=Button_Cancel
   properties {
    buttonText=Cancel
    controllerInputStyle=/FortUILibrary/Foundation/Buttons/ButtonStyles/ButtonStyle-Empty.ButtonStyle-Empty_C
    onButtonBaseClicked=()
    onButtonBaseDoubleClicked=()
    onButtonBaseHovered=()
    onButtonBaseUnhovered=()
    onSelectedChangedBase=()
    style=&ButtonStyle-Empty_C
    visibility=Visible
    pressedSlateSoundOverride {resourceObject=/Game/Sounds/Fort_UI_Sounds/FrontEnd/fort_ui_dismiss_03_Cue}
    triggeringInputAction {dataTable=&RadioInputActionTable, rowName=ReleasePicker}
   }
   slot {class=HorizontalBoxSlot, properties {horizontalAlignment=HAlign_Center, verticalAlignment=VAlign_Center, size {sizeRule=Fill, value=1}}}
  }
  CommonVisibilityWidget_3 {
   widgetClassPath=&CommonVisibilityWidgetLegacy
   parent=BottomBarOverlay
   widget=CommonVisibilityWidget_3
   properties {
    bShowForAndroid=false
    bShowForErebus=false
    bShowForGamepad=false
    bShowForIOS=false
    bShowForPS4=false
    bShowForPS5=false
    bShowForSwitch2=false
    bShowForTouch=false
    bShowForXBox=false
    bShowForXSX=false
    hiddenType=Hidden
   }
   slot {class=OverlaySlot, properties {verticalAlignment=VAlign_Center, padding=<ltrb 20 0 0 0>}}
  }
  HorizontalBox_519 {widgetClassPath=&HorizontalBox, parent=CommonVisibilityWidget_3, widget=HorizontalBox_519, slot {class=BorderSlot}}
  Text_QuickAccess_0_1 {
   widgetClassPath=&CommonTextBlock
   bInherited=true
   parent=HorizontalBox_519
   widget=Text_QuickAccess_0_1
   properties {
    style=/FortUILibrary/Styles/Typography/Foundation/TextStyle_BurbankSmall_S_White.TextStyle_BurbankSmall_S_White_C
    text=1-0
    font {
     bForceMonospaced=false
     bMaterialIsStencil=false
     fontMaterial=null
     fontObject=&BurbankSmall
     letterSpacing=0
     monospacedWidth=1
     size=18
     skewAmount=0
     typefaceFontName=Black
     outlineSettings {
      bApplyOutlineToDropShadows=false
      bMiteredCorners=false
      bSeparateFillAlpha=false
      outlineMaterial=null
      outlineSize=0
      outlineColor=<rgba 0 0 0 1>
     }
    }
    shadowOffset=<0 0>
   }
   slot {class=HorizontalBoxSlot, properties {padding=<ltrb 4 2 4 2>}}
  }
  Text_QuickAccess_Slash {
   widgetClassPath=&CommonTextBlock
   bInherited=true
   parent=HorizontalBox_519
   widget=Text_QuickAccess_Slash
   properties {
    style=&TextStyle_BurbankSmall_S_White_C
    text=/
    font {
     bForceMonospaced=false
     bMaterialIsStencil=false
     fontMaterial=null
     fontObject=&BurbankSmall
     letterSpacing=0
     monospacedWidth=1
     size=18
     skewAmount=0
     typefaceFontName=Black
     outlineSettings {
      bApplyOutlineToDropShadows=false
      bMiteredCorners=false
      bSeparateFillAlpha=false
      outlineMaterial=null
      outlineSize=0
      outlineColor=<rgba 0 0 0 1>
     }
    }
    shadowOffset=<0 0>
   }
   slot {class=HorizontalBoxSlot, properties {padding=<ltrb 4 2 4 2>}}
  }
  Text_QuickAccess_Q_Y {
   widgetClassPath=&CommonTextBlock
   bInherited=true
   parent=HorizontalBox_519
   widget=Text_QuickAccess_Q_Y
   properties {
    style=&TextStyle_BurbankSmall_S_White_C
    text=Q-Y
    font {
     bForceMonospaced=false
     bMaterialIsStencil=false
     fontMaterial=null
     fontObject=&BurbankSmall
     letterSpacing=0
     monospacedWidth=1
     size=18
     skewAmount=0
     typefaceFontName=Black
     outlineSettings {
      bApplyOutlineToDropShadows=false
      bMiteredCorners=false
      bSeparateFillAlpha=false
      outlineMaterial=null
      outlineSize=0
      outlineColor=<rgba 0 0 0 1>
     }
    }
    shadowOffset=<0 0>
   }
   slot {class=HorizontalBoxSlot, properties {padding=<ltrb 4 2 4 2>}}
  }
  CommonTextBlock_1 {
   widgetClassPath=&CommonTextBlock
   parent=HorizontalBox_519
   widget=CommonTextBlock_1
   properties {
    style=&TextStyle_BurbankSmall_S_White_C
    text=Quick Access
    font {
     bForceMonospaced=false
     bMaterialIsStencil=false
     fontMaterial=null
     fontObject=&BurbankSmall
     letterSpacing=0
     monospacedWidth=1
     size=18
     skewAmount=0
     typefaceFontName=Black
     outlineSettings {
      bApplyOutlineToDropShadows=false
      bMiteredCorners=false
      bSeparateFillAlpha=false
      outlineMaterial=null
      outlineSize=0
      outlineColor=<rgba 0 0 0 1>
     }
    }
    shadowOffset=<0 0>
   }
   slot {class=HorizontalBoxSlot, properties {padding=<ltrb 4 2 4 2>}}
  }
  Text_WheelLabel {
   widgetClassPath=&CommonTextBlock
   parent=Overlay_Main
   widget=Text_WheelLabel
   properties {
    justification=Center
    scrollStyle=&TextScrollStyle-Base_C
    text=Radio Stations
    textTransformPolicy=ToUpper
    font {
     bForceMonospaced=false
     bMaterialIsStencil=false
     fontMaterial=null
     fontObject=&Burbank
     letterSpacing=0
     monospacedWidth=1
     size=43
     skewAmount=0
     typefaceFontName=Burbank-Black
     outlineSettings {
      bApplyOutlineToDropShadows=false
      bMiteredCorners=false
      bSeparateFillAlpha=false
      outlineMaterial=&M-FontShadow
      outlineSize=3
      outlineColor=<rgba 0 0 0 0.5>
     }
    }
    shadowColorAndOpacity=<rgba 0 0 0 0.5>
    shadowOffset=<5 5>
   }
   slot {class=OverlaySlot, properties {horizontalAlignment=HAlign_Center, padding=<ltrb 0 38 0 0>}}
  }
 ]
}
