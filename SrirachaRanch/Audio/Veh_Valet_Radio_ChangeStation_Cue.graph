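# Sound cue for the station-change switch sound on the Valet vehicle radio.
# A single wave player feeds the root directly, with no randomization or modulation nodes.
# It routes through the vehicle interior sound class and uses the Valet radio concurrency set to manage overlapping instances.
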SoundCue /SrirachaRanch/Audio/Veh_Valet_Radio_ChangeStation_Cue
root=WavePlayer_0
nodes [{id=WavePlayer_0, properties {soundWaveAssetPtr=&/Veh_Valet_Radio_Switch_Change_Station_01}}]
properties {
 maxDistance=2097152
 priority=2
 soundClassObject=&Fort_Vehicles_Interior
 concurrencySet [&/Veh_Valet_Radio_Conc]
}
