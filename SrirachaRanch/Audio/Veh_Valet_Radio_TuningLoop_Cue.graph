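# Sound cue for the radio tuning loop in the Valet vehicle, playing while the radio is being tuned.
# A single looping wave player feeds the Raisin tuning loop wave straight to the output.
# It uses the Fort vehicle interior sound class, a reduced volume multiplier and a very large max distance.
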SoundCue /SrirachaRanch/Audio/Veh_Valet_Radio_TuningLoop_Cue
root=WavePlayer_0
nodes [{id=WavePlayer_0, properties {bLooping=true, soundWaveAssetPtr=&/Veh_Valet_Raisin_Tuning_Loop_01}}]
properties {maxDistance=2097152, soundClassObject=&Fort_Vehicles_Interior, volumeMultiplier=0.125}
