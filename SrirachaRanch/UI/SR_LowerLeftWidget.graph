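# Vehicle radio HUD widget on RadioPlayerWidgetBase, showing station art, name, status and track metadata.
# Radio events switch between tuning, playing, no-signal and off states, with sounds and a tuning loop faded in at a random offset.
# A timer auto-hides it: touch input hides everything, otherwise song info stays unless the radio stops or fails.
# An open station picker suppresses show/hide and enlarges the widget over a glow.
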WidgetBlueprint /SrirachaRanch/UI/SR_LowerLeftWidget
animations [
 BriefShowWholeWidget {
  binding_count=1
  length_seconds=0.3554
  object=&self:BriefShowWholeWidget
  sequence {bindings [Panel_Mainbox {
     id=#1
     possessed_class=Overlay
     tracks [
      {
       type=MovieSceneFloatTrack
       display_name=Render Opacity
       sections [{type=MovieSceneFloatSection, channels [None {type=MovieSceneScriptingFloatChannel, keys [[0,0],[5.0283,1]]}]}]
      }
      {
       type=MovieScene2DTransformTrack
       sections [{
         type=MovieScene2DTransformSection
         channels [
          Translation.X {type=MovieSceneScriptingFloatChannel, keys [[0,63],[7.1077,0]]}
          Translation.Y {type=MovieSceneScriptingFloatChannel}
          Angle {type=MovieSceneScriptingFloatChannel}
          Scale.X {type=MovieSceneScriptingFloatChannel}
          Scale.Y {type=MovieSceneScriptingFloatChannel}
          Shear.X {type=MovieSceneScriptingFloatChannel}
          Shear.Y {type=MovieSceneScriptingFloatChannel}
         ]
        }]
      }
     ]
    }]}
 }
 WholeWidgetHide {
  binding_count=1
  length_seconds=1.4354
  object=&self:WholeWidgetHide
  sequence {bindings [Panel_Mainbox {
     id=#1
     possessed_class=Overlay
     tracks [
      {
       type=MovieSceneFloatTrack
       display_name=Render Opacity
       sections [{type=MovieSceneFloatSection, channels [None {type=MovieSceneScriptingFloatChannel, keys [[4.3927,1],[28.707,0]]}]}]
      }
      {
       type=MovieScene2DTransformTrack
       sections [{
         type=MovieScene2DTransformSection
         channels [
          Translation.X {type=MovieSceneScriptingFloatChannel, keys [[0,0],[28.707,55.5353]]}
          Translation.Y {type=MovieSceneScriptingFloatChannel}
          Angle {type=MovieSceneScriptingFloatChannel}
          Scale.X {type=MovieSceneScriptingFloatChannel}
          Scale.Y {type=MovieSceneScriptingFloatChannel}
          Shear.X {type=MovieSceneScriptingFloatChannel}
          Shear.Y {type=MovieSceneScriptingFloatChannel}
         ]
        }]
      }
     ]
    }]}
 }
 StationName {
  binding_count=1
  length_seconds=0.35
  object=&self:StationName
  sequence {bindings [StationNameText {
     id=#2
     possessed_class=CommonTextBlock
     tracks [
      {
       type=MovieSceneFloatTrack
       display_name=Render Opacity
       sections [{type=MovieSceneFloatSection, end_frame=7, start_frame=0, channels [None {type=MovieSceneScriptingFloatChannel, keys [[2,0],[7,1]]}]}]
      }
      {
       type=MovieScene2DTransformTrack
       sections [{
         type=MovieScene2DTransformSection
         end_frame=7
         start_frame=0
         channels [
          Translation.X {type=MovieSceneScriptingFloatChannel, keys [[0,30],[7,0]]}
          Translation.Y {type=MovieSceneScriptingFloatChannel, const [0,1]}
          Angle {type=MovieSceneScriptingFloatChannel, const [0,1]}
          Scale.X {type=MovieSceneScriptingFloatChannel, const [1,1]}
          Scale.Y {type=MovieSceneScriptingFloatChannel, const [1,1]}
          Shear.X {type=MovieSceneScriptingFloatChannel, const [0,1]}
          Shear.Y {type=MovieSceneScriptingFloatChannel, const [0,1]}
         ]
        }]
      }
     ]
    }]}
 }
 ShowSongInfo {
  binding_count=1
  length_seconds=0.35
  object=&self:ShowSongInfo
  sequence {bindings [HorizontalBox_SongInfo {
     id=#3
     possessed_class=HorizontalBox
     tracks [
      {
       type=MovieSceneFloatTrack
       display_name=Render Opacity
       sections [{type=MovieSceneFloatSection, end_frame=7, start_frame=0, channels [None {type=MovieSceneScriptingFloatChannel, keys [[2,0],[7,1]]}]}]
      }
      {
       type=MovieScene2DTransformTrack
       sections [{
         type=MovieScene2DTransformSection
         end_frame=7
         start_frame=0
         channels [
          Translation.X {type=MovieSceneScriptingFloatChannel, keys [[0,30],[7,0]]}
          Translation.Y {type=MovieSceneScriptingFloatChannel, const [0,1]}
          Angle {type=MovieSceneScriptingFloatChannel, const [0,1]}
          Scale.X {type=MovieSceneScriptingFloatChannel, const [1,1]}
          Scale.Y {type=MovieSceneScriptingFloatChannel, const [1,1]}
          Shear.X {type=MovieSceneScriptingFloatChannel, const [0,1]}
          Shear.Y {type=MovieSceneScriptingFloatChannel, const [0,1]}
         ]
        }]
      }
     ]
    }]}
 }
 RadioStationHide_SongStay {
  binding_count=4
  length_seconds=0.35
  object=&self:RadioStationHide_SongStay
  sequence {bindings [
    Switcher_StationImage {
     id=#4
     possessed_class=CommonVisibilitySwitcher
     tracks [
      {
       type=MovieSceneFloatTrack
       display_name=Render Opacity
       sections [{type=MovieSceneFloatSection, end_frame=7, start_frame=0, channels [None {type=MovieSceneScriptingFloatChannel, keys [[0,1],[7,0]]}]}]
      }
      {
       type=MovieScene2DTransformTrack
       sections [{
         type=MovieScene2DTransformSection
         end_frame=7
         start_frame=0
         channels [
          Translation.X {type=MovieSceneScriptingFloatChannel, keys [[0,0],[7,63]]}
          Translation.Y {type=MovieSceneScriptingFloatChannel, const [0,1]}
          Angle {type=MovieSceneScriptingFloatChannel, const [0,1]}
          Scale.X {type=MovieSceneScriptingFloatChannel, const [1,1]}
          Scale.Y {type=MovieSceneScriptingFloatChannel, const [1,1]}
          Shear.X {type=MovieSceneScriptingFloatChannel, const [0,1]}
          Shear.Y {type=MovieSceneScriptingFloatChannel, const [0,1]}
         ]
        }]
      }
     ]
    }
    StatusText {
     id=#5
     possessed_class=CommonTextBlock
     tracks [
      {
       type=MovieSceneFloatTrack
       display_name=Render Opacity
       sections [{type=MovieSceneFloatSection, end_frame=7, start_frame=0, channels [None {type=MovieSceneScriptingFloatChannel, keys [[0,1],[7,0]]}]}]
      }
      {
       type=MovieScene2DTransformTrack
       sections [{
         type=MovieScene2DTransformSection
         end_frame=7
         start_frame=0
         channels [
          Translation.X {type=MovieSceneScriptingFloatChannel, keys [[0,0],[7,63]]}
          Translation.Y {type=MovieSceneScriptingFloatChannel, const [0,1]}
          Angle {type=MovieSceneScriptingFloatChannel, const [0,1]}
          Scale.X {type=MovieSceneScriptingFloatChannel, const [1,1]}
          Scale.Y {type=MovieSceneScriptingFloatChannel, const [1,1]}
          Shear.X {type=MovieSceneScriptingFloatChannel, const [0,1]}
          Shear.Y {type=MovieSceneScriptingFloatChannel, const [0,1]}
         ]
        }]
      }
     ]
    }
    StationNameText {
     id=#6
     possessed_class=CommonTextBlock
     tracks [
      {
       type=MovieSceneFloatTrack
       display_name=Render Opacity
       sections [{type=MovieSceneFloatSection, end_frame=7, start_frame=0, channels [None {type=MovieSceneScriptingFloatChannel, keys [[0,1],[7,0]]}]}]
      }
      {
       type=MovieScene2DTransformTrack
       sections [{
         type=MovieScene2DTransformSection
         end_frame=7
         start_frame=0
         channels [
          Translation.X {type=MovieSceneScriptingFloatChannel, keys [[0,0],[7,63]]}
          Translation.Y {type=MovieSceneScriptingFloatChannel, const [0,1]}
          Angle {type=MovieSceneScriptingFloatChannel, const [0,1]}
          Scale.X {type=MovieSceneScriptingFloatChannel, const [1,1]}
          Scale.Y {type=MovieSceneScriptingFloatChannel, const [1,1]}
          Shear.X {type=MovieSceneScriptingFloatChannel, const [0,1]}
          Shear.Y {type=MovieSceneScriptingFloatChannel, const [0,1]}
         ]
        }]
      }
     ]
    }
    HorizontalBox_SongInfo {
     id=#7
     possessed_class=HorizontalBox
     tracks [{
       type=MovieScene2DTransformTrack
       sections [{
         type=MovieScene2DTransformSection
         end_frame=7
         start_frame=0
         channels [
          Translation.X {type=MovieSceneScriptingFloatChannel, keys [[0,0],[7,115]]}
          Translation.Y {type=MovieSceneScriptingFloatChannel, const [0,1]}
          Angle {type=MovieSceneScriptingFloatChannel, const [0,1]}
          Scale.X {type=MovieSceneScriptingFloatChannel, const [1,1]}
          Scale.Y {type=MovieSceneScriptingFloatChannel, const [1,1]}
          Shear.X {type=MovieSceneScriptingFloatChannel, const [0,1]}
          Shear.Y {type=MovieSceneScriptingFloatChannel, const [0,1]}
         ]
        }]
      }]
    }
   ]}
 }
 PickerTransition {
  binding_count=1
  length_seconds=0.3529
  object=&self:PickerTransition
  sequence {bindings [Border_BackgroundGlow {
     id=#8
     possessed_class=Border
     tracks [
      {
       type=MovieScene2DTransformTrack
       sections [{
         type=MovieScene2DTransformSection
         end_frame=7
         start_frame=-1
         channels [
          Translation.X {type=MovieSceneScriptingFloatChannel, keys [[-0.025,0],[2,5],[4.012,-110],[6.0237,-100]]}
          Translation.Y {type=MovieSceneScriptingFloatChannel, keys [[-0.025,0],[2,5],[4.012,-110],[6.0237,-100]]}
          Scale.X {type=MovieSceneScriptingFloatChannel, keys [[-0.025,1],[2,0.975],[4.0237,1.275],[6.0237,1.25]]}
          Scale.Y {type=MovieSceneScriptingFloatChannel, keys [[-0.025,1],[2,0.975],[4.0237,1.275],[6.0237,1.25]]}
         ]
        }]
      }
      {
       type=MovieSceneColorTrack
       display_name=Brush Color
       sections [{
         type=MovieSceneColorSection
         end_frame=6
         start_frame=-1
         channels [
          Color.R {type=MovieSceneScriptingFloatChannel}
          Color.G {type=MovieSceneScriptingFloatChannel}
          Color.B {type=MovieSceneScriptingFloatChannel}
          Color.A {type=MovieSceneScriptingFloatChannel, keys [[-0.025,0],[4.0053,0],[7,0.25]]}
         ]
        }]
      }
     ]
    }]}
 }
]
graphs [
 EventGraph {nodes [
   StopAnimation(InAnimation=Get BriefShowWholeWidget() -> (BriefShowWholeWidget:Widget Animation Object Reference))
   > StopAnimation(InAnimation=Get WholeWidgetHide() -> (WholeWidgetHide:Widget Animation Object Reference))
   > StopAnimation(InAnimation=Get RadioStationHide_SongStay() -> (RadioStationHide_SongStay:Widget Animation Object Reference))
   SetRenderOpacity(
    self=[
     Get StatusText() -> (StatusText:Common Text Object Reference)
     Get Switcher_StationImage() -> (Switcher_StationImage:Common Visibility Switcher Object Reference)
     Get StationNameText() -> (StationNameText:Common Text Object Reference)
     Get HorizontalBox_SongInfo() -> (HorizontalBox_SongInfo:Horizontal Box Object Reference)
     Get Panel_Mainbox() -> (Panel_Mainbox:Horizontal Box Object Reference)
    ]
    InOpacity=1
   )
   > SetRenderTranslation(self=[Get StatusText(), Get Switcher_StationImage(), Get StationNameText(), Get HorizontalBox_SongInfo(), Get Panel_Mainbox()])
   n1 = OnAnimBackupTimerFinished() {is=title K2Node_CustomEvent}
   > Clear and Invalidate Timer by Handle(Handle=Get HideAnimFallbackTimerHandle() -> (HideAnimFallbackTimerHandle:Timer Handle Structure))
   n2 = HideWidget() {is=title K2Node_CustomEvent}
   > n3 = Branch(Condition=Get bPickerOpen() -> (bPickerOpen:bool))
   PlayAnimationForward(execute=@n3.else, InAnimation=Get CurrentHideAnimation() -> (CurrentHideAnimation:Widget Animation Object Reference))
   > n4 = Set Timer by Event(Delegate=@n1, Time=GetEndTime(self=Get CurrentHideAnimation()), bLooping=false)
   > Set HideAnimFallbackTimerHandle(HideAnimFallbackTimerHandle:Timer Handle Structure=@n4) -> (Output_Get:Timer Handle Structure)
   n5 = HideWidgetAfterTime!() -> (OutputDelegate, then, HideAnimation:Widget Animation Object Reference) {is=title K2Node_CustomEvent}
   > Set CurrentHideAnimation(CurrentHideAnimation:Widget Animation Object Reference=@n5.HideAnimation) -> (Output_Get:Widget Animation Object Reference)
   > n6 = Branch(Condition=Get bPickerOpen())
   n7 = Get!(execute:Exec=[@n9.then, @n6.then]) -> (then:Exec, else:Exec, TuningLoopAudioComponent:Audio Component Object Reference)
   > FadeOut!(
    execute
    self:Audio Component Object Reference=@n7
    FadeOutDuration:float=0.5
    FadeVolumeLevel:float=0
    FadeCurve:EAudioFaderCurve Enum=Linear
   ) -> (then)
   Clear and Invalidate Timer by Handle(execute=@n6.else, Handle=Get HideMainBoxTimer() -> (HideMainBoxTimer:Timer Handle Structure))
   > n8 = Set Timer by Event(Delegate=@n2, Time=GetRadioStatusOnScreenTime!() -> (ReturnValue:float), bLooping=false)
   > n9 = Set HideMainBoxTimer(HideMainBoxTimer:Timer Handle Structure=@n8) -> (Output_Get:Timer Handle Structure)
   Event OnPickerActivated() {is=title K2Node_Event}
   > SetVisibility(
    self=Get Switcher_StationImage()
    InVisibility="To Visibility (Boolean)"!(
     bIsVisible:bool=Get IsRadioOn() -> (IsRadioOn:bool)
     TrueVisibility:ESlateVisibility Enum=SelfHitTestInvisible
     FalseVisibility:ESlateVisibility Enum=Collapsed
    ) -> (ReturnValue:ESlateVisibility Enum=Visible)
   )
   > Reset Animation State!(execute:Exec) -> (then:Exec) {is=title K2Node_MacroInstance}
   > PlayAnimationForward(InAnimation=Get PickerTransition() -> (PickerTransition:Widget Animation Object Reference))
   Event OnPickerDeactivated() {is=title K2Node_Event}
   > PlayAnimationReverse(InAnimation=Get PickerTransition())
   > n10 = Branch(Condition=Get IsLoadingNewSource() -> (IsLoadingNewSource:bool))
   HideWidgetAfterTime!(
    execute=@n10.else
    self
    HideAnimation:Widget Animation Object Reference=Select Hide Animation!() -> (ReturnValue:Widget Animation Object Reference) {is=title K2Node_MacroInstance}
   ) -> (then) {is=title K2Node_CallFunction}
   n11 = Event OnLoadingNewSource!() -> (OutputDelegate, then, Radio:Streaming Radio Player Component Object Reference, Source:"Athena Radio Station Structure (by ref)") {is=title K2Node_Event}
   > Set IsLoadingNewSource(IsLoadingNewSource:bool=true) -> (Output_Get:bool)
   > Set IsRadioOn(IsRadioOn:bool=true) -> (Output_Get:bool)
   > Update Source Data!(execute, self, Source:Athena Radio Station Structure=@n11.Source) -> (then)
   > ShowWidget()
   > PlaySound2D(Sound=Get SoundOnRadioStarted() -> (SoundOnRadioStarted:Sound Base Object Reference))
   > n12 = Get()
   n13 = SpawnSound2D!(
    execute=@n12.else
    Sound:Sound Base Object Reference=Get SoundOnTuning() -> (SoundOnTuning:Sound Base Object Reference)
    VolumeMultiplier:float=1
    PitchMultiplier:float=1
    StartTime:float=0
    ConcurrencySettings:Sound Concurrency Object Reference
    bPersistAcrossLevelTransition:bool=false
    bAutoDestroy:bool=true
   ) -> (then, ReturnValue:Audio Component Object Reference)
   > n14 = Set TuningLoopAudioComponent(TuningLoopAudioComponent:Audio Component Object Reference=@n13) -> (Output_Get:Audio Component Object Reference)
   > FadeIn!(
    execute
    self:Audio Component Object Reference=@n14
    FadeInDuration:float=0.5
    FadeVolumeLevel:float=1
    StartTime:float=RandomFloatInRange(Min=0, Max=11.5)
    FadeCurve:EAudioFaderCurve Enum=Linear
   ) -> (then)
   Event OnSourcePlaying!() -> (OutputDelegate, then, Radio:Streaming Radio Player Component Object Reference, Source:"Athena Radio Station Structure (by ref)") {is=title K2Node_Event}
   > Branch(Condition=Get IsLoadingNewSource())
   > Set IsLoadingNewSource(IsLoadingNewSource=false)
   > UpdateStatusOfSource!(
    execute
    self
    InStyle:Common Text Style Class Reference=&/SR_Status_NowPlayingStyle.SR_Status_NowPlayingStyle_C
    InText:Text="LOCTEXT(\"Now Playing\")"
   ) -> (then)
   > SetActiveWidgetIndex(self:Common Visibility Switcher Object Reference=Get Switcher_StationImage(), Index=0)
   > HideWidgetAfterTime(HideAnimation=Select Hide Animation())
   > PlaySound2D(Sound=Get SoundOnRadioChangeStation() -> (SoundOnRadioChangeStation:Sound Base Object Reference))
   Event OnFailedToOpenSource!() -> (OutputDelegate, then, Radio:Streaming Radio Player Component Object Reference, Source:"Athena Radio Station Structure (by ref)") {is=title K2Node_Event}
   > n15 = UpdateStatusOfSource(InStyle=&/SR_Status_DefaultStyle.SR_Status_DefaultStyle_C, InText="LOCTEXT(\"No Signal! \")")
   Set IsRadioOn(execute=[@n15.then, @n17.then], IsRadioOn=false)
   > Set IsLoadingNewSource(IsLoadingNewSource=false)
   > SetVisibility(
    self=[
     Get StationNameText()
     Get Switcher_StationImage()
     Get Text_Artist() -> (Text_Artist:Common Text Object Reference)
     Get Text_Track() -> (Text_Track:Common Text Object Reference)
    ]
    InVisibility=Collapsed
   )
   > ShowWidget()
   > HideWidgetAfterTime(HideAnimation=Get WholeWidgetHide())
   > PlaySound2D(Sound=Get SoundOnRadioStopped() -> (SoundOnRadioStopped:Sound Base Object Reference))
   Event Destruct()
   > n16 = Get()
   > Stop(self:Audio Component Object Reference=@n16)
   > Set TuningLoopAudioComponent()
   Event OnRadioStopped!() -> (OutputDelegate, then, Radio:Streaming Radio Player Component Object Reference, LastSource:"Athena Radio Station Structure (by ref)") {is=title K2Node_Event}
   > n17 = UpdateStatusOfSource(InStyle=&SR_Status_DefaultStyle_C, InText="LOCTEXT(\"Radio Off\")")
   n18 = Event OnMetadataChanged!() -> (OutputDelegate, then, MediaMetadata:"Fort Media In Band Metadata Structure (by ref)") {is=title K2Node_Event}
   > n19 = Branch(Condition=Get IsLoadingNewSource())
   Update Metadata!(execute=[@n21.then, @n19.then, @n20.then], self, MediaMetadata:"Fort Media In Band Metadata Structure (by ref)"=@n18.MediaMetadata) -> (then)
   > PlayAnimationForward(InAnimation=Get ShowSongInfo() -> (ShowSongInfo:Widget Animation Object Reference))
   n20 = Branch(execute=@n19.else, Condition=Get bPickerOpen())
   n21 = SetVisibility(execute=@n20.else, self=Get Switcher_StationImage(), InVisibility=Collapsed)
   Event Construct()
   > Parent: Construct()
   > UpdateStatusOfSource(InStyle=&SR_Status_NowPlayingStyle_C, InText="LOCTEXT(\"Now Playing\")")
   StopAllAnimations(self:User Widget Object Reference)
  ]}
 ResetAnimationState {nodes [
   Inputs()
   > SetRenderOpacity(self=[Get Switcher_StationImage(), Get StatusText(), Get StationNameText(), Get HorizontalBox_SongInfo(), Get Panel_Mainbox()], InOpacity=1)
   > SetRenderTranslation(self=[Get Panel_Mainbox(), Get Switcher_StationImage(), Get StatusText(), Get StationNameText(), Get HorizontalBox_SongInfo()])
   > Clear and Invalidate Timer by Handle(Handle=Get HideMainBoxTimer())
   > Clear and Invalidate Timer by Handle(Handle=Get HideAnimFallbackTimerHandle())
   > StopAnimation(InAnimation=Get WholeWidgetHide())
   > StopAnimation(InAnimation=Get RadioStationHide_SongStay())
   > Outputs!(then:Exec)
  ]}
 SelectHideAnimation {nodes [
   Inputs!()
   n1 = GetOwningPlayer()
   Outputs!(ReturnValue:Widget Animation Object Reference=Select(
     Option 0:Widget Animation Object Reference=Get RadioStationHide_SongStay()
     Option 1:Widget Animation Object Reference=Get WholeWidgetHide()
     Index=IsUsingTouch(PlayerController=@n1)
    ) -> (ReturnValue:Widget Animation Object Reference))
  ]}
 ShowPlayingSongInfo {nodes [
   n1 = ShowPlayingSongInfo!() -> (then, Radio:Streaming Radio Player Component Object Reference) {is=title K2Node_FunctionEntry}
   > n2 = GetMediaMetadata!(execute, self:Streaming Radio Player Component Object Reference=@n1) -> (then, ReturnValue:Fort Media In Band Metadata Structure)
   > Update Metadata(MediaMetadata=@n2)
   > SetVisibility(
    self=[Get Text_Artist(), Get Text_Track()]
    InVisibility="To Visibility (Boolean)"(
     bIsVisible=IsMetadataEmpty!(MediaMetadata:"Fort Media In Band Metadata Structure (by ref)"=@n2) -> (ReturnValue:bool)
     TrueVisibility=Hidden
     FalseVisibility=SelfHitTestInvisible
    ) -> (ReturnValue=Visible)
   )
   > PlayAnimationForward(InAnimation=Get ShowSongInfo())
  ]}
 ShowWidget {nodes [
   ShowWidget() {is=title K2Node_FunctionEntry}
   > n1 = Branch(Condition=Get bPickerOpen())
   Reset Animation State(execute=@n1.else)
   > PlayAnimationForward(InAnimation=Get BriefShowWholeWidget())
  ]}
 Update Metadata {nodes [
   n1 = Update Metadata!() -> (then, MediaMetadata:"Fort Media In Band Metadata Structure (by ref)") {is=title K2Node_FunctionEntry}
   n2 = Break Fort Media in Band Metadata!(FortMediaInBandMetadata:"Fort Media In Band Metadata Structure (by ref)"=@n1) -> (Title:String, Artist:String) {is=title K2Node_BreakStruct}
   > "SetText (Text)"(self=Get Text_Track(), InText="To Text (String)"(InString=@n2))
   > "SetText (Text)"(self=Get Text_Artist(), InText="To Text (String)"(InString=@n2.Artist))
   > SetVisibility(
    self=[Get Text_Artist(), Get Text_Track()]
    InVisibility="To Visibility (Boolean)"(bIsVisible=IsMetadataEmpty(MediaMetadata=@n1), TrueVisibility=Hidden, FalseVisibility=SelfHitTestInvisible) -> (ReturnValue=Visible)
   )
  ]}
 Update Source Data {nodes [
   n1 = Update Source Data!() -> (then, Source:Athena Radio Station Structure) {is=title K2Node_FunctionEntry}
   n2 = Break Athena Radio Station!(AthenaRadioStation:"Athena Radio Station Structure (by ref)"=@n1) -> (Title:Text, StationImage:String, bIsPlayerPlaylist:bool=~) {is=title K2Node_BreakStruct}
   > "SetText (Text)"(self=Get StationNameText(), InText=@n2)
   > SetMediaURL!(
    execute
    self:Epic CMSImage Object Reference=Get StationImage() -> (StationImage:Epic CMSImage Object Reference)
    MediaUrl:String=@n2.StationImage
   ) -> (then)
   > UpdateStatusOfSource(InStyle=&/SR_Status_TuningStyle.SR_Status_TuningStyle_C, InText="LOCTEXT(\"Tuning... \")")
   > SetVisibility(self=[Get StationNameText(), Get Switcher_StationImage()], InVisibility=SelfHitTestInvisible)
   > SetVisibility(self=[Get Text_Artist(), Get Text_Track()], InVisibility=Hidden)
   > SetActiveWidgetIndex(self=Get Switcher_StationImage(), Index=1)
  ]}
 UpdateStatusOfSource {nodes [
   n1 = UpdateStatusOfSource!() -> (then, InStyle:Common Text Style Class Reference, InText:Text) {is=title K2Node_FunctionEntry}
   n2 = Get StatusText()
   > SetStyle!(execute, self:Common Text Object Reference=@n2, InStyle:Common Text Style Class Reference=@n1) -> (then)
   > "SetText (Text)"(self=@n2, InText=@n1.InText)
  ]}
]
variables [
 CurrentHideAnimation {type=WidgetAnimation, default=null}
 HideAnimFallbackTimerHandle {type=TimerHandle}
 HideMainBoxTimer {type=TimerHandle}
 IsLoadingNewSource {type=boolean}
 IsRadioOn {type=boolean}
 SoundOnRadioChangeStation {type=SoundBase, default=/SrirachaRanch/Audio/Veh_Valet_Radio_ChangeStation_Cue}
 SoundOnRadioStarted {type=SoundBase, default=/SrirachaRanch/Audio/Veh_Valet_Radio_On_Cue}
 SoundOnRadioStopped {type=SoundBase, default=/SrirachaRanch/Audio/Veh_Valet_Radio_Off_Cue}
 SoundOnTuning {type=SoundBase, default=/SrirachaRanch/Audio/Veh_Valet_Radio_TuningLoop_Cue}
 TrackAristText {type=string, default="{track} - {artist}"}
 TuningLoopAudioComponent {type=AudioComponent, default=null}
]
widgets {
 info {inheritedWidgetCount=2, parentClass=/Script/SrirachaRanchUI.RadioPlayerWidgetBase, rootWidgetClass=&Overlay, widgetCount=17}
 widgets [
  Panel_Root {widgetClassPath=&Overlay, widget=Panel_Root}
  Border_BackgroundGlow {
   widgetClassPath=&Border
   parent=Panel_Root
   widget=Border_BackgroundGlow
   properties {
    background {
     bIsDynamicallyLoaded=false
     drawAs=Image
     imageType=FullColor
     mirroring=NoMirror
     resourceName=None
     resourceObject=/Game/UI/Frontend/Store/Textures/T_RadGradient
     tiling=NoTile
     imageSize=<256 256>
     margin=<ltrb 0 0 0 0>
     outlineSettings {
      bUseBrushTransparency=false
      roundingType=HalfHeightRadius
      width=0
      color {colorUseRule=UseColor_Specified, specifiedColor=<rgba 0 0 0 0>}
      cornerRadii=<0 0 0 0>
     }
     tintColor {colorUseRule=UseColor_Specified, specifiedColor=<rgba 0.0222 0.1882 0.3216 1>}
     uVRegion {bIsValid=false, max=<0 0>, min=<0 0>}
    }
    brushColor=<rgba 1 1 1 0>
    padding=<ltrb 0 0 0 0>
   }
   slot {class=OverlaySlot, properties {horizontalAlignment=HAlign_Right, verticalAlignment=VAlign_Bottom}}
  }
  Panel_Mainbox {
   widgetClassPath=&HorizontalBox
   bIsVariable=true
   parent=Border_BackgroundGlow
   widget=Panel_Mainbox
   slot {class=BorderSlot, properties {padding=<ltrb 0 0 0 0>}}
  }
  VerticalBox_Details {
   widgetClassPath=&VerticalBox
   parent=Panel_Mainbox
   widget=VerticalBox_Details
   slot {class=HorizontalBoxSlot, properties {verticalAlignment=VAlign_Center, padding=<ltrb 0 0 8 0>}}
  }
  StatusText {
   widgetClassPath=&CommonTextBlock
   bIsVariable=true
   parent=VerticalBox_Details
   widget=StatusText
   properties {
    justification=Right
    lineHeightPercentage=0.85
    style=&SR_Status_DefaultStyle_C
    text=Status
    textTransformPolicy=ToUpper
    font {
     bForceMonospaced=false
     bMaterialIsStencil=false
     fontMaterial=null
     fontObject=&BurbankSmall
     letterSpacing=0
     monospacedWidth=1
     size=15
     skewAmount=0
     typefaceFontName=Bold
     outlineSettings {
      bApplyOutlineToDropShadows=false
      bMiteredCorners=false
      bSeparateFillAlpha=false
      outlineMaterial=&M-FontShadow
      outlineSize=3
      outlineColor=<rgba 0.0052 0.0052 0.0056 1>
     }
    }
    shadowOffset=<0 0>
   }
   slot {class=VerticalBoxSlot, properties {horizontalAlignment=HAlign_Right, verticalAlignment=VAlign_Center, padding=<ltrb 0 0 0 -3>}}
  }
  StationNameText {
   widgetClassPath=&CommonTextBlock
   bIsVariable=true
   parent=VerticalBox_Details
   widget=StationNameText
   properties {
    bAutoCollapseWithEmptyText=true
    justification=Right
    text=Station Name
    textTransformPolicy=ToUpper
    colorAndOpacity {colorUseRule=UseColor_Specified, specifiedColor=<rgba 1 1 1 0.9>}
    font {
     bForceMonospaced=false
     bMaterialIsStencil=false
     fontMaterial=null
     fontObject=&Burbank
     letterSpacing=0
     monospacedWidth=1
     size=22.5
     skewAmount=0
     typefaceFontName=BurbankRegular-Black
     outlineSettings {
      bApplyOutlineToDropShadows=false
      bMiteredCorners=false
      bSeparateFillAlpha=false
      outlineMaterial=&M-FontShadow
      outlineSize=3
      outlineColor=<rgba 0.0052 0.0052 0.0056 1>
     }
    }
   }
   slot {class=VerticalBoxSlot, properties {horizontalAlignment=HAlign_Right, padding=<ltrb 0 0 0 -5>}}
  }
  SizeBox_0 {
   widgetClassPath=&SizeBox
   parent=VerticalBox_Details
   widget=SizeBox_0
   properties {bOverride_WidthOverride=true, widthOverride=300}
   slot {class=VerticalBoxSlot}
  }
  HorizontalBox_SongInfo {widgetClassPath=&HorizontalBox, bIsVariable=true, parent=SizeBox_0, widget=HorizontalBox_SongInfo, slot {class=SizeBoxSlot}}
  VerticalBox_0 {
   widgetClassPath=&VerticalBox
   parent=HorizontalBox_SongInfo
   widget=VerticalBox_0
   slot {class=HorizontalBoxSlot, properties {size {sizeRule=Fill, value=1}}}
  }
  Text_Artist {
   widgetClassPath=&CommonTextBlock
   bIsVariable=true
   parent=VerticalBox_0
   widget=Text_Artist
   properties {
    justification=Right
    lineHeightPercentage=0.85
    scrollStyle=&TextScrollStyle-Base_C
    text="{really long track name test}"
    textTransformPolicy=ToUpper
    visibility=Hidden
    colorAndOpacity {colorUseRule=UseColor_Specified, specifiedColor=<rgba 1 0.8725 0.3209 1>}
    font {
     bForceMonospaced=false
     bMaterialIsStencil=false
     fontMaterial=null
     fontObject=&BurbankSmall
     letterSpacing=0
     monospacedWidth=1
     size=15
     skewAmount=0
     typefaceFontName=Bold
     outlineSettings {
      bApplyOutlineToDropShadows=false
      bMiteredCorners=false
      bSeparateFillAlpha=false
      outlineMaterial=&M-FontShadow
      outlineSize=3
      outlineColor=<rgba 0.0052 0.0052 0.0056 1>
     }
    }
    shadowOffset=<0 0>
   }
   slot {class=VerticalBoxSlot, properties {horizontalAlignment=HAlign_Right, padding=<ltrb 0 -3 0 0>}}
  }
  Text_Track {
   widgetClassPath=&CommonTextBlock
   bIsVariable=true
   parent=VerticalBox_0
   widget=Text_Track
   properties {
    justification=Right
    lineHeightPercentage=0.85
    scrollStyle=&TextScrollStyle-Base_C
    text=" {really long artist title test }"
    textTransformPolicy=ToUpper
    visibility=Hidden
    colorAndOpacity {colorUseRule=UseColor_Specified, specifiedColor=<rgba 1 1 1 0.8>}
    font {
     bForceMonospaced=false
     bMaterialIsStencil=false
     fontMaterial=null
     fontObject=&BurbankSmall
     letterSpacing=0
     monospacedWidth=1
     size=15
     skewAmount=0
     typefaceFontName=Bold
     outlineSettings {
      bApplyOutlineToDropShadows=false
      bMiteredCorners=false
      bSeparateFillAlpha=false
      outlineMaterial=&M-FontShadow
      outlineSize=3
      outlineColor=<rgba 0.0052 0.0052 0.0056 1>
     }
    }
    shadowOffset=<0 0>
   }
   slot {class=VerticalBoxSlot, properties {horizontalAlignment=HAlign_Right, padding=<ltrb 0 -3 0 0>}}
  }
  Switcher_StationImage {
   widgetClassPath=&CommonVisibilitySwitcher
   bIsVariable=true
   parent=Panel_Mainbox
   widget=Switcher_StationImage
   slot {class=HorizontalBoxSlot, properties {padding=<ltrb 8 0 0 0>}}
  }
  SizeBox_1 {
   widgetClassPath=&SizeBox
   parent=Switcher_StationImage
   widget=SizeBox_1
   properties {bOverride_HeightOverride=true, bOverride_WidthOverride=true, heightOverride=100, widthOverride=100}
   slot {class=CommonVisibilitySwitcherSlot, properties {horizontalAlignment=HAlign_Center, verticalAlignment=VAlign_Center}}
  }
  StationImage {
   widgetClassPath=/Script/EpicCMSUIFramework.EpicCMSImage
   bIsVariable=true
   parent=SizeBox_1
   widget=StationImage
   properties {
    visibility=SelfHitTestInvisible
    brush {
     bIsDynamicallyLoaded=false
     drawAs=Image
     imageType=FullColor
     mirroring=NoMirror
     resourceName=None
     resourceObject=/SrirachaRanch/UI/Materials/T_UI_Playlist
     tiling=NoTile
     imageSize=<128 128>
     margin=<ltrb 0 0 0 0>
     outlineSettings {
      bUseBrushTransparency=false
      roundingType=HalfHeightRadius
      width=0
      color {colorUseRule=UseColor_Specified, specifiedColor=<rgba 0 0 0 0>}
      cornerRadii=<0 0 0 0>
     }
     tintColor {colorUseRule=UseColor_Specified, specifiedColor=<rgba 1 1 1 1>}
     uVRegion {bIsValid=false, max=<0 0>, min=<0 0>}
    }
    renderTransformPivot=<0.75 0.8>
   }
   slot {class=SizeBoxSlot}
  }
  LoadingImage {
   widgetClassPath=&Image
   parent=Switcher_StationImage
   widget=LoadingImage
   properties {
    brush {
     bIsDynamicallyLoaded=false
     drawAs=Image
     imageType=NoImage
     mirroring=NoMirror
     resourceName=None
     resourceObject=/FortUILibrary/Systems/Throbber/M_V2Radio_Loading
     tiling=NoTile
     imageSize=<100 100>
     margin=<ltrb 0 0 0 0>
     outlineSettings {
      bUseBrushTransparency=false
      roundingType=HalfHeightRadius
      width=0
      color {colorUseRule=UseColor_Specified, specifiedColor=<rgba 0 0 0 0>}
      cornerRadii=<0 0 0 0>
     }
     tintColor {colorUseRule=UseColor_Specified, specifiedColor=<rgba 1 1 1 1>}
     uVRegion {bIsValid=false, max=<0 0>, min=<0 0>}
    }
    renderTransformPivot=<0 0>
   }
   slot {class=CommonVisibilitySwitcherSlot, properties {horizontalAlignment=HAlign_Center, verticalAlignment=VAlign_Center}}
  }
  Overlay_Keybind {
   widgetClassPath=&Overlay
   bInherited=true
   parent=Panel_Root
   widget=Overlay_Keybind
   properties {renderOpacity=0}
   slot {class=OverlaySlot}
  }
  ToggleActionWidget {
   widgetClassPath=&CommonActionWidget
   bInherited=true
   parent=Overlay_Keybind
   widget=ToggleActionWidget
   slot {class=OverlaySlot}
  }
 ]
}
