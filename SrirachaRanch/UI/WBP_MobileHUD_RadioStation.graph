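# Mobile HUD element widget for the radio station display, built on the BacchusHUDElement parent class.
# An inherited scale box with user-specified stretch wraps a fixed-size box that hosts an SR_LowerLeftWidget.
# The event graph holds only empty Tick and Construct events, so this Blueprint adds no logic of its own.
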WidgetBlueprint /SrirachaRanch/UI/WBP_MobileHUD_RadioStation
graphs [EventGraph {nodes [
   Event Tick()
   Event Construct()
  ]}]
variables []
widgets {
 info {inheritedWidgetCount=3, parentClass=/Script/FortniteUI.BacchusHUDElement, rootWidgetClass=&ScaleBox, widgetCount=5}
 widgets [
  ScaleBox_HUDWidgetScale {
   widgetClassPath=&ScaleBox
   bInherited=true
   widget=ScaleBox_HUDWidgetScale
   properties {stretch=UserSpecified, renderTransformPivot=<0 0>}
  }
  SizeBox_0 {
   widgetClassPath=&SizeBox
   parent=ScaleBox_HUDWidgetScale
   widget=SizeBox_0
   properties {heightOverride=70, widthOverride=215, renderTransformPivot=<0 0>}
   slot {class=ScaleBoxSlot, properties {horizontalAlignment=HAlign_Fill, verticalAlignment=VAlign_Fill}}
  }
  SR_LowerLeftWidget {
   widgetClassPath=&/SR_LowerLeftWidget
   bIsVariable=true
   parent=SizeBox_0
   widget=SR_LowerLeftWidget
   slot {class=SizeBoxSlot}
  }
  Image_Background {widgetClassPath=&Image, bInherited=true, widget=null}
  TutorialHighlight {widgetClassPath=&TutorialHighlight}
 ]
}
