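# Static widget that previews the car radio station HUD element in the HUD Layout Tool, derived from HUDLayoutToolV2_WidgetPreview.
# Shows right-aligned placeholder "Now Playing" status, station, artist and track text beside a station thumbnail, over the inherited translucent background.
# Artist and track text use scroll styles within a fixed-width box, and an inherited instances-left counter overlays the bottom-right corner.
# The event graph is empty, so all content is fixed sample data.
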WidgetBlueprint /SrirachaRanch/UI/WidgetPreview_RadioStation
graphs [EventGraph {links [], nodes []}]
variables []
widgets {
 info {inheritedWidgetCount=2, parentClass=/Script/FortniteUI.HUDLayoutToolV2_WidgetPreview, rootWidgetClass=&SizeBox, widgetCount=15}
 widgets [
  SizeBox_15 {widgetClassPath=&SizeBox, widget=SizeBox_15, properties {bOverride_HeightOverride=true, heightOverride=128}}
  Overlay_0 {widgetClassPath=&Overlay, parent=SizeBox_15, widget=Overlay_0, slot {class=SizeBoxSlot}}
  Border_Background {
   widgetClassPath=&Border
   bInherited=true
   parent=Overlay_0
   widget=Border_Background
   properties {brushColor=<rgba 0 0 0 0.2>, padding=<ltrb 10 0 0 0>}
   slot {class=OverlaySlot, properties {horizontalAlignment=HAlign_Fill, verticalAlignment=VAlign_Fill}}
  }
  HorizontalBox_0 {
   widgetClassPath=&HorizontalBox
   parent=Border_Background
   widget=HorizontalBox_0
   slot {class=BorderSlot, properties {padding=<ltrb 10 0 0 0>}}
  }
  VerticalBox_Details {
   widgetClassPath=&VerticalBox
   parent=HorizontalBox_0
   widget=VerticalBox_Details
   slot {class=HorizontalBoxSlot, properties {verticalAlignment=VAlign_Center, padding=<ltrb 0 0 8 0>}}
  }
  StatusText {
   widgetClassPath=&CommonTextBlock
   parent=VerticalBox_Details
   widget=StatusText
   properties {
    justification=Right
    lineHeightPercentage=0.85
    style=&/SR_Status_NowPlayingStyle.SR_Status_NowPlayingStyle_C
    text=Now Playing
    textTransformPolicy=ToUpper
    colorAndOpacity {colorUseRule=UseColor_Specified, specifiedColor=<rgba 0 1 0.9843 1>}
    font {
     bForceMonospaced=false
     bMaterialIsStencil=false
     fontMaterial=null
     fontObject=&BurbankSmall
     letterSpacing=0
     monospacedWidth=1
     size=15
     skewAmount=0
     typefaceFontName=Bold
     outlineSettings {
      bApplyOutlineToDropShadows=false
      bMiteredCorners=false
      bSeparateFillAlpha=false
      outlineMaterial=&M-FontShadow
      outlineSize=3
      outlineColor=<rgba 0.0052 0.0052 0.0056 1>
     }
    }
    shadowOffset=<0 0>
   }
   slot {class=VerticalBoxSlot, properties {horizontalAlignment=HAlign_Right, verticalAlignment=VAlign_Center, padding=<ltrb 0 0 0 -3>}}
  }
  StationName {
   widgetClassPath=&CommonTextBlock
   parent=VerticalBox_Details
   widget=StationName
   properties {
    justification=Right
    text=Party Royale
    textTransformPolicy=ToUpper
    colorAndOpacity {colorUseRule=UseColor_Specified, specifiedColor=<rgba 1 1 1 0.9>}
    font {
     bForceMonospaced=false
     bMaterialIsStencil=false
     fontMaterial=null
     fontObject=&Burbank
     letterSpacing=0
     monospacedWidth=1
     size=22.5
     skewAmount=0
     typefaceFontName=BurbankRegular-Black
     outlineSettings {
      bApplyOutlineToDropShadows=false
      bMiteredCorners=false
      bSeparateFillAlpha=false
      outlineMaterial=&M-FontShadow
      outlineSize=3
      outlineColor=<rgba 0.0052 0.0052 0.0056 1>
     }
    }
    shadowOffset=<0 0>
   }
   slot {class=VerticalBoxSlot, properties {horizontalAlignment=HAlign_Right, padding=<ltrb 0 0 0 -5>}}
  }
  SizeBox_ArtistTrackInfo {
   widgetClassPath=&SizeBox
   parent=VerticalBox_Details
   widget=SizeBox_ArtistTrackInfo
   properties {bOverride_WidthOverride=true, widthOverride=300}
   slot {class=VerticalBoxSlot, properties {padding=<ltrb 0 -3 0 0>}}
  }
  HorizontalBox_3 {widgetClassPath=&HorizontalBox, parent=SizeBox_ArtistTrackInfo, widget=HorizontalBox_3, slot {class=SizeBoxSlot}}
  VerticalBox_1 {
   widgetClassPath=&VerticalBox
   parent=HorizontalBox_3
   widget=VerticalBox_1
   slot {class=HorizontalBoxSlot, properties {size {sizeRule=Fill, value=1}}}
  }
  Text_Artist {
   widgetClassPath=&CommonTextBlock
   parent=VerticalBox_1
   widget=Text_Artist
   properties {
    justification=Right
    scrollStyle=&TextScrollStyle-Base_C
    text=Epic Games
    textTransformPolicy=ToUpper
    colorAndOpacity {colorUseRule=UseColor_Specified, specifiedColor=<rgba 1 0.8725 0.3209 1>}
    font {
     bForceMonospaced=false
     bMaterialIsStencil=false
     fontMaterial=null
     fontObject=&BurbankSmall
     letterSpacing=0
     monospacedWidth=1
     size=15
     skewAmount=0
     typefaceFontName=Bold
     outlineSettings {
      bApplyOutlineToDropShadows=false
      bMiteredCorners=false
      bSeparateFillAlpha=false
      outlineMaterial=&M-FontShadow
      outlineSize=3
      outlineColor=<rgba 0.0052 0.0052 0.0056 1>
     }
    }
    shadowOffset=<0 0>
   }
   slot {class=VerticalBoxSlot}
  }
  Text_Track {
   widgetClassPath=&CommonTextBlock
   parent=VerticalBox_1
   widget=Text_Track
   properties {
    justification=Right
    lineHeightPercentage=0.85
    scrollStyle=&TextScrollStyle-Base_C
    text=When the wind blows
    textTransformPolicy=ToUpper
    colorAndOpacity {colorUseRule=UseColor_Specified, specifiedColor=<rgba 1 1 1 0.8>}
    font {
     bForceMonospaced=false
     bMaterialIsStencil=false
     fontMaterial=null
     fontObject=&BurbankSmall
     letterSpacing=0
     monospacedWidth=1
     size=15
     skewAmount=0
     typefaceFontName=Bold
     outlineSettings {
      bApplyOutlineToDropShadows=false
      bMiteredCorners=false
      bSeparateFillAlpha=false
      outlineMaterial=&M-FontShadow
      outlineSize=3
      outlineColor=<rgba 0.0052 0.0052 0.0056 1>
     }
    }
    shadowOffset=<0 0>
   }
   slot {class=VerticalBoxSlot, properties {horizontalAlignment=HAlign_Right, padding=<ltrb 0 -3 0 0>}}
  }
  SizeBox_StationThumbnail {
   widgetClassPath=&SizeBox
   parent=HorizontalBox_0
   widget=SizeBox_StationThumbnail
   properties {bOverride_HeightOverride=true, bOverride_WidthOverride=true, heightOverride=100, widthOverride=100}
   slot {class=HorizontalBoxSlot, properties {horizontalAlignment=HAlign_Center, verticalAlignment=VAlign_Center, padding=<ltrb 8 8 8 8>}}
  }
  Image_StationThumbnail {
   widgetClassPath=&Image
   parent=SizeBox_StationThumbnail
   widget=Image_StationThumbnail
   properties {
    visibility=SelfHitTestInvisible
    brush {
     bIsDynamicallyLoaded=false
     drawAs=Image
     imageType=NoImage
     mirroring=NoMirror
     resourceName=None
     resourceObject=/SrirachaRanch/UI/Materials/T_UI_Valet_Station1
     tiling=NoTile
     imageSize=<100 100>
     margin=<ltrb 0 0 0 0>
     outlineSettings {
      bUseBrushTransparency=false
      roundingType=HalfHeightRadius
      width=0
      color {colorUseRule=UseColor_Specified, specifiedColor=<rgba 0 0 0 0>}
      cornerRadii=<0 0 0 0>
     }
     tintColor {colorUseRule=UseColor_Specified, specifiedColor=<rgba 1 1 1 1>}
     uVRegion {bIsValid=false, max=<0 0>, min=<0 0>}
    }
    renderTransformPivot=<0.75 0.8>
   }
   slot {class=SizeBoxSlot}
  }
  TextBlock_InstancesLeftCounter {
   widgetClassPath=&CommonTextBlock
   bInherited=true
   parent=Overlay_0
   widget=TextBlock_InstancesLeftCounter
   properties {
    justification=Center
    textTransformPolicy=ToUpper
    font {
     bForceMonospaced=false
     bMaterialIsStencil=false
     fontMaterial=null
     fontObject=&BurbankSmall
     letterSpacing=0
     monospacedWidth=1
     size=16
     skewAmount=0
     typefaceFontName=Black
     outlineSettings {
      bApplyOutlineToDropShadows=false
      bMiteredCorners=false
      bSeparateFillAlpha=false
      outlineMaterial=null
      outlineSize=0
      outlineColor=<rgba 0 0 0 1>
     }
    }
    shadowOffset=<0 0>
   }
   slot {class=OverlaySlot, properties {horizontalAlignment=HAlign_Right, verticalAlignment=VAlign_Bottom, padding=<ltrb 0 0 4 0>}}
  }
 ]
}
